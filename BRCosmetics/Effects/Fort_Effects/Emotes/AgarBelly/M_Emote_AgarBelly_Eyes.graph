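# Unlit translucent Niagara sprite and mesh particle material that shows animated eyes for the AgarBelly emote.
# A FlipBook samples the eyes atlas on UV1, with its phase driven by looping time and per-particle speed, start index and offset dynamic parameters.
# Particle color tints the result through an emissive adjustment function.
# Opacity combines particle and flipbook alpha with a UV0 vertical falloff and depth fade.
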Material /BRCosmetics/Effects/Fort_Effects/Emotes/AgarBelly/M_Emote_AgarBelly_Eyes
graph {
 properties {
  bAutomaticallySetUsageInEditor=false
  bUsedWithNiagaraMeshParticles=true
  bUsedWithNiagaraSprites=true
  blendMode=BLEND_Translucent
  shadingModel=MSM_Unlit
  usedShadingModels=Unlit
 }
 nodes [
  n1 = ParticleColor()
  n2 = DynamicParameter() {defaultValue=<rgba 1 0 0.5 1>, paramNames [AnimationSpeed, StartIndex, Offset, Param4]}
  n3 = MaterialFunctionCall(Speed=@n2.AnimationSpeed) {materialFunction=&TimeWithSpeedVariable}
  n4 = MaterialFunctionCall(
   "Animation  Phase (0-1)"=Add(A=LinearInterpolate(A=@n2.StartIndex, Alpha=@n3.Frac Time), B=@n2.Offset)
   Number of Rows=ScalarParameter() {defaultValue=8, parameterName="Number of Rows (S)"}
   Number of Columns=ScalarParameter() {defaultValue=8, parameterName="Number of Columns (S)"}
   Texture=TextureObject() {texture=&/T_Emote_AgarBelly_EyesV2}
   UVs=TextureCoordinate() {coordinateIndex=1}
  ) {materialFunction=&FlipBook}
  n5 = ComponentMask(None=TextureCoordinate()) {g=true}
  n6 = Multiply(A=Power(Base=OneMinus(None=@n5)) {constExponent=1}) {constB=3}
  DepthFade(Opacity=Multiply(A=Multiply(A=@n1.A, B=@n4.Alpha), B=Saturate(None=@n6))) {fadeDistanceDefault=5}
  MaterialFunctionCall(Emissive=Multiply(A=@n1.RGB, B=@n4.Result)) {materialFunction=&MF_FX_EmissiveAdjustments}
  ScalarParameter() {parameterName=Roughness}
  ScalarParameter() {defaultValue=1, parameterName=Specular}
  ScalarParameter() {defaultValue=1, parameterName=Metallic}
  VectorParameter() {parameterName=Base Color}
 ]
}
